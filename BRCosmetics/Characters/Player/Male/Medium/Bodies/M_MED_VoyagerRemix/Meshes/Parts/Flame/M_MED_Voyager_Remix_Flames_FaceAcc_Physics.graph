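# Physics asset for the flame face accessory on the Voyager Remix male medium character cosmetic.
# It defines a sphere body on the dyn_flames bone, offset slightly from the bone, and a small capsule body on the head bone.
# A single constraint ties dyn_flames to head with matching limits on both swing axes and twist, bounding how far the flames can rotate relative to the head.
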PhysicsAsset /BRCosmetics/Characters/Player/Male/Medium/Bodies/M_MED_VoyagerRemix/Meshes/Parts/Flame/M_MED_Voyager_Remix_Flames_FaceAcc_Physics
bodies {dyn_flames [{shapeType=Sphere, radius=5, center=<1 0 0>}], head [{shapeType=Capsule, length=1, radius=1}]}
constraints [{bone1Name=dyn_flames, bone2Name=head, swing1LimitDegrees=60, swing2LimitDegrees=60, twistLimitDegrees=60}]
