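# Animation Blueprint for the flame face accessory of the male medium Voyager Remix cosmetic body, adding secondary physics motion to the flames.
# It copies the pose from the attached parent mesh, then runs a RigidBody node in root-bone space using the flame physics asset and slight upward gravity.
# AnimDynamics nodes for the flame bones and a targeting-based vector select remain in the graph but do not reach the output, and the event graph does nothing.
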AnimBlueprint /BRCosmetics/Characters/Player/Male/Medium/Bodies/M_MED_VoyagerRemix/Meshes/Parts/Flame/M_MED_Voyager_Remix_Flames_FaceAcc_AnimBP
target_skeleton=&/M_MED_Voyager_Remix_Flames_FaceAcc_Skeleton
graphs [
 AnimGraph {nodes [
   Anim Dynamics - Bone: flame_1(Alpha=1, ExternalForce="0,0,0") {
    angularDampingOverride=1
    angularSpringConstant=1500
    bAngularSpring=true
    bGravityOverrideInSimSpace=true
    bOverrideAngularDamping=true
    gravityScale=0.1
    simulationSpace=RootRelative
    alphaBoolBlend {blendOption=Linear}
    boundBone {boneName=flame_1}
    physicsBodyDefinitions [{
      sphereCollisionRadius=0
      boundBone {boneName=flame_1}
      boxExtents=<10 10 10>
      constraintSetup {angularLimitsMax=<0 5 5>, angularLimitsMin=<0 -5 -5>, angularTarget=<0 0 0>}
      localJointOffset=<-10 0 0>
     }]
   }
   n1 = Anim Dynamics - Bone: flame_2(Alpha=1, ExternalForce="0,0,0") {
    angularDampingOverride=0.999
    angularSpringConstant=300
    bAngularSpring=true
    bGravityOverrideInSimSpace=true
    bOverrideAngularDamping=true
    gravityScale=0.3
    simulationSpace=RootRelative
    alphaBoolBlend {blendOption=Linear}
    boundBone {boneName=flame_2}
    componentLinearAccScale=<0.5 1 0.4>
    componentLinearVelScale=<0.5 1 0.4>
    physicsBodyDefinitions [{
      sphereCollisionRadius=0
      boundBone {boneName=flame_2}
      boxExtents=<10 10 10>
      constraintSetup {angularLimitsMax=<5 30 50>, angularLimitsMin=<-5 -30 -50>, angularTarget=<0 0 0>}
      localJointOffset=<-5 0 0>
     }]
   }
   n2 = Anim Dynamics - Bone: flame_2_l(ComponentPose=@n1, Alpha=1, ExternalForce="0,0,0") {
    angularDampingOverride=1
    angularSpringConstant=1500
    bAngularSpring=true
    bGravityOverrideInSimSpace=true
    bOverrideAngularDamping=true
    gravityScale=0.1
    simulationSpace=RootRelative
    alphaBoolBlend {blendOption=Linear}
    boundBone {boneName=flame_2_l}
    componentLinearAccScale=<0.3 0.3 0.3>
    componentLinearVelScale=<0.3 0.3 0.3>
    physicsBodyDefinitions [{
      sphereCollisionRadius=0
      boundBone {boneName=flame_2_l}
      boxExtents=<10 20 20>
      constraintSetup {angularLimitsMax=<5 30 10>, angularLimitsMin=<-5 0 -10>, angularTarget=<0 0 0>}
      localJointOffset=<-10 0 0>
     }]
   }
   Anim Dynamics - Bone: flame_2_r(ComponentPose=@n2, Alpha=1, ExternalForce="0,0,0") {
    angularDampingOverride=1
    angularSpringConstant=300
    bAngularSpring=true
    bGravityOverrideInSimSpace=true
    bOverrideAngularDamping=true
    gravityScale=0.1
    simulationSpace=RootRelative
    alphaBoolBlend {blendOption=Linear}
    boundBone {boneName=flame_2_r}
    componentLinearAccScale=<0.3 0.3 0.3>
    componentLinearVelScale=<0.3 0.3 0.3>
    physicsBodyDefinitions [{
      sphereCollisionRadius=0
      boundBone {boneName=flame_2_r}
      boxExtents=<10 20 20>
      constraintSetup {angularLimitsMax=<5 30 5>, angularLimitsMin=<-5 0 -10>, angularTarget=<0 0 0>}
      localJointOffset=<10 0 0>
     }]
   }
   n3 = RigidBody(ComponentPose=Local To Component(LocalPose=Copy Pose From Mesh() {bUseAttachedParent=true})) {
    bOverrideWorldGravity=true
    overridePhysicsAsset=&/M_MED_Voyager_Remix_Flames_FaceAcc_Physics
    simulationSpace=BaseBoneSpace
    alphaBoolBlend {blendOption=Linear}
    baseBoneRef {boneName=root}
    componentLinearAccScale=<0.5 1 1>
    componentLinearVelScale=<0.5 1 1>
    overrideWorldGravity {z=1}
   }
   Output Pose(Result=Component To Local(ComponentPose=@n3)) {layerGroup=None}
   SelectVector(A="0.25,0.25,0.25", B="1,1,1", bPickA=Get bIsTargeting())
  ]}
 EventGraph {nodes [
   TryGetPawnOwner()
   Event BlueprintUpdateAnimation()
  ]}
]
variables []
